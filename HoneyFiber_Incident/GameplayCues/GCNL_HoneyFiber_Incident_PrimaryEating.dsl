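; Looping gameplay cue that plays eating audio on a player pawn's current weapon mesh for the HoneyFiber Incident item's primary action.
; On start, it spawns a first- or third-person sound by local view and sets a repeating timer that fades out and respawns it.
; Local ammo change events feed remaining ammo into the sound's "Spray_Loop" parameter.
; On removal, it plays an optional ending sound, fades the loop and clears the timer.
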(event Custom|Local_Ammo
  (params (LocalCount :type Integer) (LocalRemaining :type Integer))
  (bind _audiocomponent_sprayloop_heal (Variables|Default|GetAudioComponent_SprayLoop_Heal)
    (:then
      (Audio|Parameter|SetFloatParameter _audiocomponent_sprayloop_heal "Spray_Loop" (Math|Conversions|ToFloat(Integer) LocalRemaining) :!class /Script/AudioExtensions.AudioParameterControllerInterface))
    (:else)))

(event EventOnLoopingStart(EitherVFX-preferthisversion)
  (params (MyTarget :type Actor%20Object%20Reference) (Parameters :type Gameplay%20Cue%20Parameters%20Structure%20%28by%20ref%29) (ParticleComponents :type Array%20of%20FXSystem%20Component%20Object%20References) (AudioComponents :type Array%20of%20Audio%20Component%20Object%20References))
  (bind _asfort_player_pawn (Utilities|Casting|CastToFortPlayerPawn MyTarget))
  (Variables|Default|SetPlayerPawn _asfort_player_pawn)
  (CallFunction|StartEatingEffects)
  (bind _returnvalue (Utilities|Time|SetTimerbyEvent (delegate Custom|EatingLoop) 1e+01 true))
  (Variables|Default|SetEatingTimer _returnvalue))

(event Custom|EatingLoop
  (CallFunction|StartEatingEffects))

(event EventOnRemoval(EitherVFX-preferthisversion)
  (params (MyTarget :type Actor%20Object%20Reference) (Parameters :type Gameplay%20Cue%20Parameters%20Structure%20%28by%20ref%29) (ParticleComponents :type Array%20of%20FXSystem%20Component%20Object%20References) (AudioComponents :type Array%20of%20Audio%20Component%20Object%20References) (BurstCameraShakeInstance :type Camera%20Shake%20Base%20Object%20Reference) (BurstDecalInstance :type Decal%20Actor%20Object%20Reference))
  (CallFunction|StopEatingEffects)
  (Utilities|Time|ClearandInvalidateTimerbyHandle 0 (Variables|Default|GetEatingTimer)))

(fn StartEatingEffects
  (bind _playerpawn (Variables|Default|GetPlayerPawn))
  (bind _returnvalue (Weapon|GetCurrentWeapon _playerpawn))
  (bind _audiocomponent_sprayloop_heal (Variables|Default|GetAudioComponent_SprayLoop_Heal)
    (:then
      (Audio|Components|Audio|FadeOut _audiocomponent_sprayloop_heal (Variables|Sound|GetFade%20Out%20Duration) 0.0 "Linear" :!class /Script/Engine.AudioComponent))
    (:else
      _))
  (bind _returnvalue_2 (Audio|SpawnSoundAttached (select (Pawn|IsLocallyViewed _playerpawn) (Variables|Sound|GetEatingSound1p) (Variables|Sound|GetEatingSound3p)) (Variables|/Script/FortniteGame.FortWeapon|GetWeaponMesh _returnvalue)))
  (Variables|Default|SetAudioComponent_SprayLoop_Heal _returnvalue_2)
  (EventDispatchers|BindEventtoOnLocalAmmoChanged _returnvalue (EventDispatchers|CreateEvent :!event Local_Ammo)))

(fn StopEatingEffects
  (bind _playerpawn (Variables|Default|GetPlayerPawn)
    (:then
      (bind _eatingremoval (Variables|Sound|GetEatingRemoval)
        (:then
          (Audio|SpawnSoundAttached _eatingremoval (Variables|/Script/FortniteGame.FortWeapon|GetWeaponMesh (Weapon|GetCurrentWeapon _playerpawn))))
        (:else
          _))
      (bind _audiocomponent_sprayloop_heal (Variables|Default|GetAudioComponent_SprayLoop_Heal)
        (:then
          (Audio|Components|Audio|FadeOut _audiocomponent_sprayloop_heal (Variables|Sound|GetFade%20Out%20Duration) 0.0 "Linear" :!class /Script/Engine.AudioComponent))
        (:else)))
    (:else
      _)))

(fn ConstructionScript)
